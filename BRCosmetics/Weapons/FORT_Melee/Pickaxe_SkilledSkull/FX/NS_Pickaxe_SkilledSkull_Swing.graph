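# Niagara swing trail for the Skilled Skull pickaxe, built from two ribbon emitters that spawn animation trails between bones or sockets.
# Trail_Base has two ribbon renderers and computes separate offset position, color and width attributes for a detail layer.
# Trail_Tip is a thin ribbon that narrows and fades over its short life.
# User.Variant switches base texture multiplier and tip color ramp; system fade and time-of-day values modulate both emitters.
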NiagaraSystem /BRCosmetics/Weapons/FORT_Melee/Pickaxe_SkilledSkull/FX/NS_Pickaxe_SkilledSkull_Swing
dependencies {
 usedDataInterfaces [&NiagaraDataInterfaceCurve, &NiagaraDataInterfaceColorCurve, &NiagaraDataInterfaceSocketReader]
 usedDynamicInputs []
 usedModules [
  &self:SystemSpawnScript.SystemScriptSource.SystemScriptGraph.NiagaraNodeAssignment_0.SetVariables_#1
  &self:SystemSpawnScript.SystemScriptSource.SystemScriptGraph.NiagaraNodeAssignment_1.SetVariables_#2
  &SystemState
  &self:SystemSpawnScript.SystemScriptSource.SystemScriptGraph.NiagaraNodeAssignment_2.SetVariables_#3
 ]
}
emitters {
 Trail_Base {
  emitterSpawnScript [/SetVariables(
    Emitter.MainTextureMultiplier:NiagaraFloat=&MergeHelper_FloatSwitch(0=1, 1=0, 2=?, 3=?, Index=&MakeIntFromFloat(Float=$User.Variant), Num=::NewEnumerator0)
    Emitter.SpawnVariant:NiagaraBool=&SetBoolByIntComparison(A=&MakeIntFromFloat(Float=$User.Variant), B=1, Comparison Type=::NewEnumerator4)
   ) {isSetParameters=true}]
  emitterUpdateScript [
   &EmitterState()
   &FN_SpawnAnimTrail()
  ]
  particleSpawnScript [
   &FN_InitializeAnimTrailBoneIndecies() {module=FN_InitializeAnimTrailBoneIndecies001}
   &FN_InitializeAnimTrail(
    Color=&MultiplyLinearColor(
     Linear Color=&LerpLinearColors(StartColor=<rgba 1 1 1 1>, EndColor=$System.TOD, LerpFactor=0.6)
     Scale Factor=&MakeVector4FromVectorAndFloat("Vector (RGB)"=<1 1 1>, "Float (Alpha)"=&Multiply_Float(A=$System.TrailFade, B=0.5))
    )
    Lifetime=0.16
   ) {module=FN_InitializeAnimTrail001}
   /SetVariables(Particles.DetailPos:NiagaraPosition=&AddVectorToPosition(Position=$Particles.Position, Vector=&TransformVector(Vector=<0 0 11>, Destination Space=::Simulation))) {isSetParameters=true}
  ]
  particleUpdateScript [
   &FN_UpdateAnimTrail()
   &ScaleColor(
    Scale Alpha=&Multiply_Float(
     A=&FloatFromCurve(
      FloatCurve=DataInterface() {propertyValues {Curve {keys [
          <at 0 1 RCIM_Cubic>
          {arriveTangent=-1, interpMode=RCIM_Cubic, leaveTangent=-1, time=0.7575, value=0.2215}
          {interpMode=RCIM_Cubic, leaveTangent=-0.9767, time=1, value=0}
         ]}}}
      Scale Curve=$System.TrailFade
     )
     B=&FNTimeOfDayScalar(Night Scale=0.6, Morning Scale=0.65, Evening Scale=0.6)
    )
    Scale RGB=&MakeVectorFromLinearColorRGB(Color=<rgba 1 1 1 1>)
    ScaleA=true
    ScaleRGB=true
   ) {module=ScaleColor002}
   /SetVariables(
    Particles.DetailColor:LinearColor=&MultiplyLinearColor(
     Linear Color=$System.TOD
     Scale Factor=&MakeVector4_FromLinearColor(LinearColor=&MakeLinearColorFromVectorAndFloat(
       "Vector (RGB)"=&MakeVectorFromLinearColorRGB(Color=&ColorFromCurve(ColorCurve=DataInterface() {propertyValues {
           AlphaCurve {keys [<at 0 1 RCIM_Cubic>]}
           BlueCurve {keys [<at 0.007 1>,<at 0.3983 0.1205>]}
           GreenCurve {keys [<at 0.007 1>,<at 0.3983 0>]}
           RedCurve {keys [<at 0.007 1>,<at 0.3983 1>]}
          }}))
       "Float (Alpha)"=&Multiply_Float(
        A=&FloatFromCurve(
         FloatCurve=DataInterface() {propertyValues {Curve {keys [
             <at 0 1 RCIM_Cubic>
             {arriveTangent=-3.2556, interpMode=RCIM_Cubic, leaveTangent=-3.2556, time=0.3, value=0}
             {interpMode=RCIM_Cubic, leaveTangent=-2.5, time=0.3, value=0.0233}
            ]}}}
         Scale Curve=0
        )
        B=&FNTimeOfDayScalar(Night Scale=1, Morning Scale=1, Evening Scale=1)
       )
      ))
    )
    Particles.DetailWidth:NiagaraFloat=&Multiply_Float(A=$Particles.Initial.RibbonWidth, B=0.6)
   ) {isSetParameters=true}
  ]
  rendererClasses [&NiagaraRibbonRendererProperties]
  renderers [
   {rendererClass=&NiagaraRibbonRendererProperties}
   {rendererClass=&NiagaraRibbonRendererProperties, rendererIndex=1}
  ]
 }
 Trail_Tip {
  emitterUpdateScript [
   &EmitterState()
   &FN_SpawnAnimTrail(AnimTrailWidthMode=::FromFirst)
  ]
  particleSpawnScript [
   &FN_InitializeAnimTrailBoneIndecies(First Bone or Socket Index=1, Second Bone or Socket Index=0) {module=FN_InitializeAnimTrailBoneIndecies001}
   &FN_InitializeAnimTrail(
    AnimTrailWidth=0.05
    Color=&MultiplyLinearColor(
     Linear Color=<rgba 1.1882 1.5 0.6406 1>
     Scale Factor=&MakeVector4FromVectorAndFloat("Vector (RGB)"=<1 1 1>, "Float (Alpha)"=&Multiply_Float(A=3, B=$System.TrailFade))
    )
    Lifetime=0.16
   ) {module=FN_InitializeAnimTrail001}
   &SystemLocation(MaskSpawnGroup=false, Offset=<0 0 50>) {enabled=false}
  ]
  particleUpdateScript [
   &FN_UpdateAnimTrail()
   &ScaleRibbonWidth(Ribbon Width Scale=&FloatFromCurve(FloatCurve=DataInterface() {propertyValues {Curve {keys [
         {arriveTangent=-1.2579, interpMode=RCIM_Cubic, time=0, value=1}
         {arriveTangent=-0.9664, interpMode=RCIM_Cubic, leaveTangent=-0.9664, tangentMode=RCTM_User, time=1, value=0}
        ]}}}))
   &ScaleColor(
    Scale Alpha=&FloatFromCurve(
     FloatCurve=DataInterface() {propertyValues {Curve {keys [<at 0 1>, {arriveTangent=-0.2139, time=0.5, value=0}, <at 0.5 0.0233>]}}}
     Scale Curve=&Multiply_Float(
      A=$System.TrailFade
      B=&FNTimeOfDayScalar(Night Scale=0.8, Morning Scale=0.35, Evening Scale=0.35)
     )
    )
    Scale RGB=&MakeVectorFromLinearColorRGB(Color=&MultiplyLinearColor(
      Linear Color=&MergeHelper_ColorSwitch(
       0=<rgba 1 1 1 1>
       1=&ColorFromCurve(ColorCurve=DataInterface() {propertyValues {
          AlphaCurve {keys [<at 0 1 RCIM_Cubic>,<at 1 0 RCIM_Cubic>]}
          BlueCurve {keys [<at 0 1.5 RCIM_Cubic>,<at 0.6679 0.1295 RCIM_Cubic>]}
          GreenCurve {keys [<at 0 1.4515 RCIM_Cubic>,<at 0.6679 0 RCIM_Cubic>]}
          RedCurve {keys [<at 0 0.0938 RCIM_Cubic>,<at 0.6679 1 RCIM_Cubic>]}
         }})
       2=?
       3=?
       Index=&MakeIntFromFloat(Float=$User.Variant)
       Num=::NewEnumerator0
      )
      Scale Factor=&MakeVector4_FromLinearColor(LinearColor=$System.TOD)
     ))
    ScaleA=true
    ScaleRGB=true
   ) {module=ScaleColor001}
  ]
  rendererClasses [&NiagaraRibbonRendererProperties]
  renderers [{rendererClass=&NiagaraRibbonRendererProperties}]
 }
}
summary {
 systemName=NS_Pickaxe_SkilledSkull_Swing
 emitters [
  {emitterName=Trail_Base, rendererClasses [&NiagaraRibbonRendererProperties]}
  {emitterName=Trail_Tip, rendererClasses [&NiagaraRibbonRendererProperties]}
 ]
 userVariables [
  User.RibbonIDOffset {type=NiagaraInt32}
  User.SwingSectionIndex {type=NiagaraInt32}
  User.Variant {type=NiagaraFloat, defaultValue=0, description=""}
 ]
}
