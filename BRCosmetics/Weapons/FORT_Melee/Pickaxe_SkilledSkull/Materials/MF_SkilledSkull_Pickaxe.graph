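# Material function for the Skilled Skull pickaxe that overlays a stylized, glowing blade onto the input material attributes through a texture-based blade mask.
# A view-dependent gleam, Perlin distortion and an edge gradient broken up by panning swirl noise form a wrapped value that samples separate core and shell curve-atlas color ramps.
# The mask-selected color is tinted by time of day and drives both base color and emissive.
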MaterialFunction /BRCosmetics/Weapons/FORT_Melee/Pickaxe_SkilledSkull/Materials/MF_SkilledSkull_Pickaxe
graph {nodes [
  n1 = NamedRerouteUsage()
  n2 = Constant3Vector() {constant=<rgba 2 0.5 1.5 1>}
  n3 = Multiply(
   A=DotProduct(A=Transform(None=@n2) {transformSourceType=TRANSFORMSOURCE_Local}, B=CameraVectorWS())
   B=ScalarParameter() {defaultValue=1, group=Blade, parameterName=GleamScale}
  )
  n4 = Multiply(A=TextureCoordinate(), B=AppendVector(A=Constant() {r=3}, B=Constant() {r=0.7})) {constB=10}
  n5 = Multiply(
   A=ConstantBiasScale(None=TextureSample(UVs=@n4) {samplerType=SAMPLERTYPE_Masks, texture=/BRCosmetics/Weapons/FORT_Melee/Pickaxe_SkilledSkull/Textures/T_PerlinNoise_RGB}.B) {bias=-0.5, scale=2}
   B=ScalarParameter() {defaultValue=0.05, group=Blade, parameterName=DistortionNoiseScale}
  )
  n6 = Power(
   Base=NamedRerouteUsage()
   Exp=LinearInterpolate(
    A=ScalarParameter() {defaultValue=0.2, group=Blade, parameterName=Core_EdgeThickness}
    B=ScalarParameter() {defaultValue=0.3, group=Blade, parameterName=Shell_EdgeThickness}
    Alpha=NamedRerouteUsage()
   )
  )
  n7 = Constant() {r=0.1}
  n8 = Constant() {r=-0.08}
  n9 = NamedRerouteUsage()
  n10 = Panner(Coordinate=Multiply(A=TextureCoordinate()) {constB=10}, Speed=AppendVector(A=Constant(), B=LinearInterpolate(A=@n7, B=@n8, Alpha=@n9)))
  n11 = Subtract(
   A=Multiply(A=OneMinus(None=@n6)) {constB=2}
   B=Multiply(
    A=TextureSample(UVs=@n10) {samplerType=SAMPLERTYPE_LinearColor, texture=/BRCosmetics/Weapons/FORT_Melee/Pickaxe_SkilledSkull/Textures/T_SwirlNoise}.R
    B=ScalarParameter() {defaultValue=0.1, group=Blade, parameterName=EdgeNoiseScale}
   )
  )
  n12 = Frac(None=Add(A=Add(A=@n3, B=@n5), B=Saturate(None=@n11)))
  n13 = CurveAtlasRowParameter(CurveTime=@n12) {
   atlas=/BRCosmetics/Characters/Player/Male/Medium/Bodies/M_MED_Keychain/Textures/A_KeyChainAtlas
   curve=/BRCosmetics/Weapons/FORT_Melee/Pickaxe_SkilledSkull/Textures/C_SkilledSkull_Pickaxe_InnerBlade
   defaultValue=6
   group=Blade
   parameterName=RemapColorCurve_Core
  }
  n14 = VectorParameter() {defaultValue=<rgba 0.7356 1 0.5 1>}
  n15 = CurveAtlasRowParameter(CurveTime=@n12) {
   atlas=&A_KeyChainAtlas
   curve=/BRCosmetics/Characters/Player/Female/Medium/Bodies/F_MED_SkilledSkull/Textures/C_SkilledSkull_Body
   defaultValue=4
   group=Blade
   parameterName=RemapColorCurve_Shell
  }
  n16 = Multiply(
   A=LinearInterpolate(A=Multiply(A=@n13, B=@n14.RGB), B=@n15, Alpha=NamedRerouteUsage())
   B=MaterialFunctionCall(
    Color_Morning=VectorParameter() {group=TOD Tints, parameterName=Color_Morning, defaultValue=<rgba 0.8 0.8 0.8 1>}.RGB
    Color_Day=VectorParameter() {group=TOD Tints, parameterName=Color_Day, defaultValue=<rgba 1 1 1 1>}.RGB
    Color_Evening=VectorParameter() {group=TOD Tints, parameterName=Color_Evening, defaultValue=<rgba 0.9 0.6938 0.675 1>}.RGB
    Color_Night=VectorParameter() {group=TOD Tints, parameterName=Color_Night, defaultValue=<rgba 0.4333 0.4 0.8 1>}.RGB
   ) {materialFunction=&Color_Time_of_Day}.Result
  )
  FunctionOutput(None=BlendMaterialAttributes(
    A=NamedRerouteUsage()
    B=SetMaterialAttributes!(MaterialAttributes=@n1, BaseColor=@n16, EmissiveColor=@n16) {attributeSetTypes [#1, #2]}
    Alpha=NamedRerouteUsage()
   ))
  NamedRerouteDeclaration(None=FunctionInput() {inputName=InMat, inputType=FunctionInput_MaterialAttributes}) {name=InMat, nodeColor=<rgba 0 1 0.5078 1>}
  n17 = TextureSample() {samplerType=SAMPLERTYPE_Masks, texture=/BRCosmetics/Weapons/FORT_Melee/Pickaxe_SkilledSkull/Textures/T_SkilledSkull_Pickaxe_EdgeGradient}
  NamedRerouteDeclaration(None=@n17.G) {name=EdgeGradient, nodeColor=<rgba 0 0.1328 1 1>}
  NamedRerouteDeclaration(None=Add(
    A=NamedRerouteDeclaration(None=@n17.R) {name=OuterMask, nodeColor=<rgba 1 0.1875 0 1>}
    B=NamedRerouteDeclaration(None=@n17.B) {name=InnerMask, nodeColor=<rgba 0.4531 1 0 1>}
   )) {name=BladeMask, nodeColor=<rgba 1 0 0.7734 1>}
 ]}
