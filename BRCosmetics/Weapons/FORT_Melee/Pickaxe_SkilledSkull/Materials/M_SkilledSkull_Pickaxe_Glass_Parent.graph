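# Translucent, lit, two-sided parent material for the glass of the SkilledSkull pickaxe.
# Noise normals, projected from bounds-normalized pre-skinned position, perturb a Fresnel term driving masked emissive and opacity, while packed textures can override specular, metallic, roughness and opacity.
# Final opacity is multiplied by an FX green-channel mask that is zeroed on front faces, leaving only back faces visible.
# World position offset combines LOD and black hole effects before tech-art overrides.
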Material /BRCosmetics/Weapons/FORT_Melee/Pickaxe_SkilledSkull/Materials/M_SkilledSkull_Pickaxe_Glass_Parent
graph {
 properties {
  bAutomaticallySetUsageInEditor=false
  bEnableResponsiveAA=true
  bScreenSpaceReflections=true
  bUseMaterialAttributes=true
  bUsedWithMeshDeformer=true
  bUsedWithMorphTargets=true
  bUsedWithSkeletalMesh=true
  blendMode=BLEND_Translucent
  ditherOpacityMask=true
  pixelDepthOffsetMode=PDOM_Legacy
  translucencyDirectionalLightingIntensity=5
  translucencyLightingMode=TLM_SurfacePerPixelLighting
  twoSided=true
  usedShadingModels=Default Lit
 }
 nodes [
  n1 = LinearInterpolate(
   A=SphereMask(
    A=TextureSample() {samplerType=SAMPLERTYPE_Masks, texture=/BRCosmetics/Weapons/FORT_Melee/Pickaxe_SkilledSkull/Textures/T_SkilledSkull_Pickaxe_FX}.G
    B=Constant() {r=0.9}
   ) {attenuationRadius=0.05}
   Alpha=Saturate(None=TwoSidedSign())
  ) {constB=0}
  Multiply(B=@n1)
  n2 = VectorParameter() {parameterName=WorldPostionOfBlackHole, defaultValue=<rgba -30000 25536 1952 0>}
  n3 = MaterialFunctionCall!(
   World Position=WorldPosition().XYZ
   DistanceCullFadeAnimationOffset=MaterialFunctionCall() {materialFunction=/Game/Packages/Fortress_SharedMaterials/Base_Material_Functions/PickupEffectDistanceCullOverride}.Result
   Use actor center instead of pivot point
  ) {materialFunction=/Game/Packages/Fortress_SharedMaterials/Base_Material_Functions/LODEffect}
  n4 = WorldPosition()
  n5 = Add(
   A=MaterialFunctionCall!(WorldPostionOfBlackHole=AppendVector(A=@n2.RGB, B=@n2.RGB), World Position=Add(A=@n3.Result, B=@n4.XYZ)) {materialFunction=/Game/Packages/Fortnite_UniqueMaterials_1/BlackHoleFunction}.Result
   B=@n3.Result
  )
  n6 = MaterialFunctionCall(Float3=TextureSampleParameter2D() {parameterName=Input S, texture=/Game/Global/Textures/Default/Blanks/T_BlackUserInterface}.RGB) {materialFunction=&BreakOutFloat3Components}
  n7 = StaticBoolParameter() {group=Specular, parameterName=SwizzleRoughnessToGreen}
  n8 = MaterialFunctionCall(Float3=MaterialFunctionCall(X=@n6.R, Y=StaticSwitch(True=@n6.B, False=@n6.B, Value=@n7), Z=StaticSwitch(True=@n6.G, False=@n6.G, Value=@n7)) {materialFunction=&MakeFloat3}.Result) {materialFunction=&BreakOutFloat3Components}
  n9 = ScalarParameter() {defaultValue=1, parameterName=Fresnel Exponent}
  n10 = ObjectBounds()
  n11 = PreSkinnedLocalBounds()
  n12 = Saturate(None=Floor(None=ScalarParameter() {parameterName=Use Pre-Skinned Bounds, sliderMax=1}))
  n13 = ComponentMask(None=Divide(A=PreSkinnedPosition(), B=LinearInterpolate(A=@n10, B=@n11.Half Extents, Alpha=@n12))) {b=true, g=true}
  n14 = ScalarParameter() {defaultValue=0.2, parameterName=WaterTiling}
  n15 = TextureSampleParameter2D(UVs=VertexInterpolator(VS=Multiply(A=@n13, B=@n14)).PS) {
   parameterName=Noise Normal
   samplerSource=SSM_Wrap_WorldGroupSettings
   samplerType=SAMPLERTYPE_Normal
   texture=/Game/Global/Textures/Default/Noise/T_Noise_Normal_01
  }
  n16 = VectorParameter() {parameterName=Normal Intensity}
  n17 = TextureSampleParameter2D() {
   group=Mask Texture
   parameterName=Normals
   samplerSource=SSM_Wrap_WorldGroupSettings
   samplerType=SAMPLERTYPE_Normal
   texture=/Engine/EngineMaterials/FlatNormal
  }
  n18 = Normalize(VectorInput=Add(A=Multiply(A=@n15.RGB, B=@n16.RGB), B=@n17.RGB))
  n19 = Transform(None=@n18)
  n20 = Clamp(None=Multiply(A=Fresnel(ExponentIn=@n9, Normal=@n19), B=ScalarParameter() {defaultValue=1, parameterName=Glass thickness}))
  n21 = LinearInterpolate(
   A=ScalarParameter() {parameterName=Fresnel Inner Emissive}
   B=ScalarParameter() {defaultValue=0.85, parameterName=Fresnel Outer Emissive}
   Alpha=@n20
  )
  n22 = VectorParameter() {parameterName=EmissiveColor}
  n23 = TextureSampleParameter2D() {parameterName=Input FX, texture=&T_BlackUserInterface}
  n24 = ScalarParameter() {parameterName=Add Custom Opacity, sortPriority=10}
  n25 = ScalarParameter() {parameterName=Fresnel Inner Transparency}
  n26 = ScalarParameter() {defaultValue=0.85, parameterName=Fresnel Outer Transparency}
  n27 = LinearInterpolate(
   A=Add(A=Multiply(A=@n23.B, B=@n24), B=LinearInterpolate(A=@n25, B=@n26, Alpha=@n20))
   B=@n23.B
   Alpha=ScalarParameter() {parameterName=Use Custom Opacity}
  )
  n28 = MakeMaterialAttributes!(
   BaseColor=VectorParameter() {parameterName=Base Color}.RGB
   Metallic=LinearInterpolate(A=ScalarParameter() {defaultValue=0.1, parameterName=Metallic}, B=@n8.G, Alpha=ScalarParameter() {parameterName=Use Custom Metallic})
   Specular=LinearInterpolate(A=ScalarParameter() {defaultValue=1, parameterName=Specular}, B=@n8.R, Alpha=ScalarParameter() {parameterName=Use Custom Specular})
   Roughness=LinearInterpolate(A=ScalarParameter() {parameterName=Roughness}, B=@n8.B, Alpha=ScalarParameter() {parameterName=Use Custom Roughness})
   Anisotropy
   EmissiveColor=Multiply(
    A=Multiply(A=@n21, B=@n22.RGB)
    B=TextureSampleParameter2D() {parameterName=EmissiveMask, samplerType=SAMPLERTYPE_Masks, texture=&T_BlackMask}.RGB
   )
   Opacity=Multiply(A=@n27, B=@n1)
   OpacityMask=@n27
   Normal=@n18
   Tangent
   WorldPositionOffset
   SubsurfaceColor
   ClearCoat
   ClearCoatRoughness
   AmbientOcclusion
   Refraction
   CustomizedUVs_0
   CustomizedUVs_1
   CustomizedUVs_2
   CustomizedUVs_3
   CustomizedUVs_4
   CustomizedUVs_5
   CustomizedUVs_6
   CustomizedUVs_7
   PixelDepthOffset
   ShadingModel
   Displacement
  )
  Reroute(None=MaterialFunctionCall(In=SetMaterialAttributes!(MaterialAttributes=@n28, World Position Offset=@n5) {attributeSetTypes [#1]}) {materialFunction=&MF_TechArtOverrides}.Result)
 ]
}
